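# Material function for the Octane Courgette01 car skin that overlays animated strawberry and cheese patterns on the base color.
# Two UV1 tile grids of different density scroll vertically at different speeds, with a sine hash randomly shifting each row horizontally.
# Per-cell noise lookups vary brightness and gate the rotating sprite UVs inside square masks.
# Cheese alpha blends the layers, and the skin mask's green channel places the result over the base color.
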MaterialFunction /VehicleCosmetics/Bodies/Octane/Skins/OctaneCourgette01/MF_Octane_Cougette01
graph {nodes [
  n1 = TextureCoordinate() {coordinateIndex=1}
  n2 = ComponentMask(None=@n1) {r=true}
  n3 = ComponentMask(None=@n1) {g=true}
  n4 = Add(A=@n3, B=Multiply(A=Time(), B=Constant() {r=0.18}))
  n5 = Constant2Vector() {g=11, r=11}
  n6 = Multiply(A=Frac(None=AppendVector(A=@n2, B=@n4)), B=@n5.G)
  n7 = ComponentMask(None=@n6) {g=true}
  n8 = Floor(None=Add(A=@n7, B=Constant() {r=0.01}) {constB=0.01})
  n9 = Multiply(A=Sine(None=Multiply(A=@n8) {constB=12.364}) {period=0.5}) {constB=6849821.5}
  n10 = Frac(None=@n9)
  n11 = AppendVector(A=Add(A=ComponentMask(None=@n6) {r=true}, B=Multiply(A=@n10) {constB=1.5}), B=@n7)
  n12 = Frac(None=@n11)
  n13 = Subtract(A=@n12, B=Constant() {r=0.5})
  n14 = Abs(None=@n13)
  n15 = SmoothStep(Min=Constant() {r=0.4}, Max=Constant() {r=0.39}, Value=Max(A=ComponentMask(None=@n14) {r=true}, B=ComponentMask(None=@n14) {g=true}))
  ComponentMask(None=@n15) {r=true}
  Constant() {r=8}
  Constant() {r=3}
  n16 = Add(A=@n3, B=Multiply(A=Time(), B=Constant() {r=0.09}))
  n17 = Constant2Vector() {g=20, r=20}
  n18 = Multiply(A=Frac(None=AppendVector(A=@n2, B=@n16)), B=@n17.G)
  n19 = ComponentMask(None=@n18) {g=true}
  n20 = Constant() {r=0.01}
  n21 = Multiply(A=Floor(None=Add(A=@n19, B=@n20) {constB=0.01})) {constB=12.364}
  Divide(A=@n21) {constB=18}
  n22 = Multiply(A=Sine(None=@n21)) {constB=6849821.5}
  Floor(None=@n22)
  Floor(None=@n9)
  n23 = ComponentMask(None=@n18) {r=true}
  n24 = Multiply(A=Frac(None=@n22)) {constB=1.5}
  n25 = Add(A=AppendVector(A=Add(A=@n23, B=@n24), B=@n19), B=Constant()) {constB=0}
  n26 = TextureSample(UVs=Divide(A=Floor(None=@n25), B=@n17.G)) {texture=&/Noise_Generic_01_Pack}
  n27 = Frac(None=@n25)
  n28 = Constant() {r=1}
  n29 = Constant()
  n30 = Multiply(A=Rotator(Coordinate=@n27) {speed=2}, B=If(A=@n26.R, B=Constant() {r=0.2}, "A > B"=@n28, "A == B"=@n28, "A < B"=@n29))
  n31 = Constant() {r=0.4}
  n32 = Constant() {r=0.39}
  n33 = Abs(None=Subtract(A=@n27, B=Constant() {r=0.5}))
  n34 = Max(A=ComponentMask(None=@n33) {r=true}, B=ComponentMask(None=@n33) {g=true})
  n35 = Multiply(
   A=Clamp(None=Multiply(A=@n26.R) {constB=3}) {maxDefault=1.5}
   B=Multiply(
    A=TextureSample(UVs=@n30) {texture=&/Strawberrie}.RGB
    B=SmoothStep(Min=@n31, Max=@n32, Value=@n34)
   )
  )
  n36 = TextureSample(UVs=Divide(A=Floor(None=@n11), B=@n17.G)) {texture=&Noise_Generic_01_Pack}
  n37 = Constant() {r=0}
  n38 = TextureSample(UVs=Multiply(A=Rotator(Coordinate=@n12) {speed=2}, B=If(A=@n36.G, B=@n37, "A > B"=@n28, "A == B"=@n28, "A < B"=@n29) {equalsThreshold=0})) {texture=&/Cheese}
  n39 = Multiply(A=Clamp(None=Multiply(A=@n36.R) {constB=8}) {maxDefault=1.5}, B=Multiply(A=@n38.RGB, B=@n15))
  FunctionOutput(None=LinearInterpolate(
    A=TextureSample() {texture=&/Skin_Octane_Courgette01_BC}.RGB
    B=LinearInterpolate(A=@n35, B=@n39, Alpha=@n38.A)
    Alpha=TextureSample() {texture=&/Skin_OctaneCourgette01_RGB}.G
   ))
  Panner(Coordinate=Subtract(A=@n6, B=@n5)) {speedY=2.4}
  Panner(Coordinate=@n13) {speedY=0.1}
 ]}
